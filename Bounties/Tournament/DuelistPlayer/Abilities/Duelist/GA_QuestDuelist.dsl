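; Gameplay ability for a Versus Battle bounty quest that pits the player against a tracked opponent, with urgent-quest HUD and spectator messages.
; Without a target, the server queues the player and a timer falls back to a bot-candidate duel.
; Duel start applies tag effects, a shrinking tracking circle, death listeners and a time warning.
; Target kill, poaching, primary-hunter death or expiry resolves the quest, then cleanup delays InDuel tag removal for late damage.
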(event Ability|EventActivateAbility
  ;[0 0 0] Activate
  (bind _asfort_player_pawn (Utilities|Casting|CastToFortPlayerPawn (Ability|GetAvatarActorfromActorInfo)))
  (GameplayEffects|ApplyGameplayEffectToSelf :self (Ability|GetFortAbilitySystemComponent _asfort_player_pawn) :GameplayEffectClass (Variables|Default|GetQueuedForDUel) :Level 0.0)
  (if (Networking|IsServer)
    (Utilities|IsValid (CallFunction|GetTrackedHunterBountyTarget)
      (:Is%20Valid
        (CallFunction|PlayerDuelStarted))
      (:Is%20Not%20Valid
        (EventDispatchers|BindEventtoOnPlayerDuelEventStarted (Class|FortPlayerPawn|GetTransientQuestsComponent _asfort_player_pawn) (EventDispatchers|CreateEvent :!event Custom|PlayerDuelStarted))
        (bind _returnvalue (Utilities|Time|SetTimerbyEvent (EventDispatchers|CreateEvent :!event Custom|StartDuelWithBotCandidate) 9e+01))
        (Variables|Default|SetQueuedForDuelTimer _returnvalue)
        (Variables|Default|SetbQueuedForDuel true)
        (CallFunction|ShowQueuingMessage (Class|FortPlayerPawn|GetTransientQuestsComponent _asfort_player_pawn))))))

(event Custom|PlayerDuelStarted
  (bind _tracked_hunter_bounty_target_pawn (Variables|Default|GetTracked%20Hunter%20Bounty%20Target%20Pawn))
  ;[0 0 0] OnPostAbilityInitialization
  (bind _asfort_player_pawn (Utilities|Casting|CastToFortPlayerPawn (Ability|GetAvatarActorfromActorInfo)))
  (GameplayEffects|ApplyGameplayEffectToSelf :self (Ability|GetFortAbilitySystemComponent _asfort_player_pawn) :GameplayEffectClass (Variables|Default|GetInDuel) :Level 0.0)
  (Ability|RemoveGameplayEffectFromOwnerWithGrantedTags (Variables|Default|GetRemove%20Queued%20Tag))
  (Ability|HasAuthority
    (:True
      (Variables|Default|SetbHasDuelStarted true)
      (Utilities|IsValid (Variables|FortTransientQuestGameplayAbility|GetQuestItemDefinition)
        (:Is%20Valid
          (bind _urgent_quest_data (CallFunction|CreateUrgentQuestData))
          (Class|AthenaHUDContext|ShowUrgentQuestMessage (BlueprintContext|GetAthenaHUDContext) _urgent_quest_data)
          (Utilities|Time|SetTimerbyEvent (EventDispatchers|CreateEvent :!event Custom|OneMinuteLeft) (- (Math|Float|SelectFloat (Fortnite|GetNumericCVar "Fort.Bounty.TotalQuestTime") (ScalableFloat|GetValueatLevel (Variables|Default|GetTotalQuestTime)) (and (> (Fortnite|GetNumericCVar "Fort.Bounty.TotalQuestTime") 0.0) (Editor|InEditororPIE))) 59.0))
          (EventDispatchers|AssignOnPawnDiedDelegate (CallFunction|GetTrackedHunterBountyTarget) (EventDispatchers|CreateEvent :!event Custom|OnTargetDied_Bind))
          (if (not (CallFunction|GetIsPrimaryHunter))
            (EventDispatchers|BindEventtoOnPawnDiedDelegate (CallFunction|GetTrackedPrimaryHunter) (EventDispatchers|CreateEvent :!event Custom|OnPrimaryHunter_Died)))
          (EventDispatchers|AssignOnUrgentQuestTimeExpired (Class|FortPlayerControllerAthena|GetTransientQuestsComponent (Utilities|Casting|CastToFortPlayerControllerAthena (Fortnite|GetFortPlayerControllerfromActor (Ability|GetAvatarActorfromActorInfo)))) (EventDispatchers|CreateEvent :!event Custom|OnUrgentQuestTimeExpired_Event))
          (CallFunction|AddTrackingCircleInfo (ScalableFloat|GetValueatLevel (Variables|Default|GetMaxCenterOffset)) (ScalableFloat|GetValueatLevel (Variables|Default|GetMinimumRadius)) (ScalableFloat|GetValueatLevel (Variables|Default|GetMinimumRadius)) 1e+01 (ScalableFloat|GetValueatLevel (Variables|Default|GetRadiusShrinkAmountPerUpdate)) (Variables|Default|GetCircleColor))
          ;[0 0 0] Apply duel tag if target is a bot
          (if (GameplayTags|HasMatchingGameplayTag _tracked_hunter_bounty_target_pawn "(TagName=\"Athena.IsAPhoebe\")")
            (Ability|GameplayEffect|AddLooseGameplayTags _tracked_hunter_bounty_target_pawn (Variables|Default|GetInDuelTag)))
          (bind _asfort_urgent_quest_item_definition (Utilities|Casting|CastToFortUrgentQuestItemDefinition (Variables|FortTransientQuestGameplayAbility|GetQuestItemDefinition)))
          (Variables|Default|SetQuestItemDefEventTag (Utilities|Struct|BreakUrgentQuestData (Class|FortUrgentQuestItemDefinition|GetUrgentQuestData _asfort_urgent_quest_item_definition)))
          (bind _structout (Utilities|Struct|SetmembersinUrgentQuestData (Class|FortUrgentQuestItemDefinition|GetUrgentQuestData _asfort_urgent_quest_item_definition) (Math|Float|SelectFloat (Fortnite|GetNumericCVar "Fort.Bounty.TotalQuestTime") (ScalableFloat|GetValueatLevel (Variables|Default|GetTotalQuestTime)) (and (> (Fortnite|GetNumericCVar "Fort.Bounty.TotalQuestTime") 0.0) (Editor|InEditororPIE))) (Utilities|Text|FormatText (.EventSubtitle (Utilities|Struct|BreakUrgentQuestData (Class|FortUrgentQuestItemDefinition|GetUrgentQuestData _asfort_urgent_quest_item_definition))) (Class|FortPlayerControllerGameplay|GetSettingsFilteredPlayerName (Utilities|Casting|CastToFortPlayerControllerGameplay (Fortnite|GetFortPlayerControllerfromActor (Ability|GetAvatarActorfromActorInfo))) (CallFunction|GetTrackedHunterBountyTarget))) (select (CallFunction|GetIsPrimaryHunter) "" (Utilities|Text|FormatText (.EventSubtitleSecondary (Utilities|Struct|BreakUrgentQuestData (Class|FortUrgentQuestItemDefinition|GetUrgentQuestData _asfort_urgent_quest_item_definition))) (Class|FortPlayerControllerGameplay|GetSettingsFilteredPlayerName (Utilities|Casting|CastToFortPlayerControllerGameplay (Fortnite|GetFortPlayerControllerfromActor (Ability|GetAvatarActorfromActorInfo))) (CallFunction|GetTrackedPrimaryHunter)))) (Utilities|Text|FormatText (.EventSubtitle (Utilities|Struct|BreakUrgentQuestData (Class|FortUrgentQuestItemDefinition|GetUrgentQuestData _asfort_urgent_quest_item_definition))) (Class|FortPlayerControllerGameplay|GetSettingsFilteredPlayerName (Utilities|Casting|CastToFortPlayerControllerGameplay (Fortnite|GetFortPlayerControllerfromActor (Ability|GetAvatarActorfromActorInfo))) (CallFunction|GetTrackedHunterBountyTarget))) (CallFunction|GetTrackedHunterBountyTarget) (CallFunction|GetTrackedPrimaryHunter) (Class|FortPlayerStateZone|GetSocialAvatarBrush (CallFunction|GetTrackedHunterBountyTarget) true)))
          (CallFunction|StartUrgentQuestEvent _structout)
          (CallFunction|SpectatorShowUrgentQuestMessage _structout)
          (if (Variables|Default|GetbQueuedForDuel)
            (Utilities|Time|ClearandInvalidateTimerbyHandle (Variables|Default|GetQueuedForDuelTimer))
            (CallFunction|EndQueueEvent "(TagName=\"Athena.Quests.QueuedForDuel\")")
            (Variables|Default|SetbQueuedForDuel false)))))))

(event EventOnPostAbilityInitialization
  (bind _tracked_hunter_bounty_target_pawn (Variables|Default|GetTracked%20Hunter%20Bounty%20Target%20Pawn))
  ;[0 0 0] OnPostAbilityInitialization
  (if (Variables|Default|GetbHasDuelStarted)
    (bind _urgent_quest_data (CallFunction|CreateUrgentQuestData))
    (Class|AthenaHUDContext|ShowUrgentQuestMessage (BlueprintContext|GetAthenaHUDContext) _urgent_quest_data)
    (Utilities|Time|SetTimerbyEvent (EventDispatchers|CreateEvent :!event Custom|OneMinuteLeft) (- (Math|Float|SelectFloat (Fortnite|GetNumericCVar "Fort.Bounty.TotalQuestTime") (ScalableFloat|GetValueatLevel (Variables|Default|GetTotalQuestTime)) (and (> (Fortnite|GetNumericCVar "Fort.Bounty.TotalQuestTime") 0.0) (Editor|InEditororPIE))) 59.0))
    (EventDispatchers|AssignOnPawnDiedDelegate (CallFunction|GetTrackedHunterBountyTarget) (EventDispatchers|CreateEvent :!event Custom|OnTargetDied_Bind))
    (if (not (CallFunction|GetIsPrimaryHunter))
      (EventDispatchers|BindEventtoOnPawnDiedDelegate (CallFunction|GetTrackedPrimaryHunter) (EventDispatchers|CreateEvent :!event Custom|OnPrimaryHunter_Died)))
    (EventDispatchers|AssignOnUrgentQuestTimeExpired (Class|FortPlayerControllerAthena|GetTransientQuestsComponent (Utilities|Casting|CastToFortPlayerControllerAthena (Fortnite|GetFortPlayerControllerfromActor (Ability|GetAvatarActorfromActorInfo)))) (EventDispatchers|CreateEvent :!event Custom|OnUrgentQuestTimeExpired_Event))
    (CallFunction|AddTrackingCircleInfo (ScalableFloat|GetValueatLevel (Variables|Default|GetMaxCenterOffset)) (ScalableFloat|GetValueatLevel (Variables|Default|GetMinimumRadius)) (ScalableFloat|GetValueatLevel (Variables|Default|GetMinimumRadius)) 1e+01 (ScalableFloat|GetValueatLevel (Variables|Default|GetRadiusShrinkAmountPerUpdate)) (Variables|Default|GetCircleColor))
    ;[0 0 0] Apply duel tag if target is a bot
    (if (GameplayTags|HasMatchingGameplayTag _tracked_hunter_bounty_target_pawn "(TagName=\"Athena.IsAPhoebe\")")
      (Ability|GameplayEffect|AddLooseGameplayTags _tracked_hunter_bounty_target_pawn (Variables|Default|GetInDuelTag)))
    (bind _asfort_urgent_quest_item_definition (Utilities|Casting|CastToFortUrgentQuestItemDefinition (Variables|FortTransientQuestGameplayAbility|GetQuestItemDefinition)))
    (Variables|Default|SetQuestItemDefEventTag (Utilities|Struct|BreakUrgentQuestData (Class|FortUrgentQuestItemDefinition|GetUrgentQuestData _asfort_urgent_quest_item_definition)))
    (bind _structout (Utilities|Struct|SetmembersinUrgentQuestData (Class|FortUrgentQuestItemDefinition|GetUrgentQuestData _asfort_urgent_quest_item_definition) (Math|Float|SelectFloat (Fortnite|GetNumericCVar "Fort.Bounty.TotalQuestTime") (ScalableFloat|GetValueatLevel (Variables|Default|GetTotalQuestTime)) (and (> (Fortnite|GetNumericCVar "Fort.Bounty.TotalQuestTime") 0.0) (Editor|InEditororPIE))) (Utilities|Text|FormatText (.EventSubtitle (Utilities|Struct|BreakUrgentQuestData (Class|FortUrgentQuestItemDefinition|GetUrgentQuestData _asfort_urgent_quest_item_definition))) (Class|FortPlayerControllerGameplay|GetSettingsFilteredPlayerName (Utilities|Casting|CastToFortPlayerControllerGameplay (Fortnite|GetFortPlayerControllerfromActor (Ability|GetAvatarActorfromActorInfo))) (CallFunction|GetTrackedHunterBountyTarget))) (select (CallFunction|GetIsPrimaryHunter) "" (Utilities|Text|FormatText (.EventSubtitleSecondary (Utilities|Struct|BreakUrgentQuestData (Class|FortUrgentQuestItemDefinition|GetUrgentQuestData _asfort_urgent_quest_item_definition))) (Class|FortPlayerControllerGameplay|GetSettingsFilteredPlayerName (Utilities|Casting|CastToFortPlayerControllerGameplay (Fortnite|GetFortPlayerControllerfromActor (Ability|GetAvatarActorfromActorInfo))) (CallFunction|GetTrackedPrimaryHunter)))) (Utilities|Text|FormatText (.EventSubtitle (Utilities|Struct|BreakUrgentQuestData (Class|FortUrgentQuestItemDefinition|GetUrgentQuestData _asfort_urgent_quest_item_definition))) (Class|FortPlayerControllerGameplay|GetSettingsFilteredPlayerName (Utilities|Casting|CastToFortPlayerControllerGameplay (Fortnite|GetFortPlayerControllerfromActor (Ability|GetAvatarActorfromActorInfo))) (CallFunction|GetTrackedHunterBountyTarget))) (CallFunction|GetTrackedHunterBountyTarget) (CallFunction|GetTrackedPrimaryHunter) (Class|FortPlayerStateZone|GetSocialAvatarBrush (CallFunction|GetTrackedHunterBountyTarget) true)))
    (CallFunction|StartUrgentQuestEvent _structout)
    (CallFunction|SpectatorShowUrgentQuestMessage _structout)
    (if (Variables|Default|GetbQueuedForDuel)
      (Utilities|Time|ClearandInvalidateTimerbyHandle (Variables|Default|GetQueuedForDuelTimer))
      (CallFunction|EndQueueEvent "(TagName=\"Athena.Quests.QueuedForDuel\")")
      (Variables|Default|SetbQueuedForDuel false))))

(event Custom|OnUrgentQuestTimeExpired_Event
  (params (UrgentQuestData :type Urgent%20Quest%20Data%20Structure%20%28by%20ref%29))
  (bind _urgentquestdata (Utilities|Struct|MakeUrgentQuestData :EventTag "(TagName=\"SpecialEvent.HuntingPlayer.IsHunter\")" :VisualStyleTag "(GameplayTags=)" :TotalEventTime 6e+01 :bDisplayTimer true :bQuestCompleteOnTimeExpiration false :bQuickIntro false :bQuickOutro false :bRewardPersist false :bDisplayMultiStepExpanded true :EventTitle "NSLOCTEXT(\"[FBF308A94707D7D5BFB54B980182BE05]\", \"1AA7EF2E4FDF762AC51423A623FF3114\", \"Versus Battle Failed\")" :EventSubtitle "NSLOCTEXT(\"[FBF308A94707D7D5BFB54B980182BE05]\", \"D157131A4FB5261F72ACC3B868D0093A\", \"Time Expired\")" :EventDescription "NSLOCTEXT(\"[FBF308A94707D7D5BFB54B980182BE05]\", \"FEFBEA9443EB52CAD602AEB261B70C7F\", \"Description\")" :EventDescriptionOverrideTag "(TagName=\"\")" :SortPriority 0 :bHideQuestProgressBar false :bDisplayAllObjectivesIndividually true :QuestProgressTextFormat "LOCGEN_FORMAT_ORDERED(INVTEXT(\"`{{0}} / `{{1}}\"), INVTEXT(\"CurrentProgress\"), INVTEXT(\"MaxProgress\"))" :bReverseProgressBar true :bShowBountyThreatInformation false :bShowBountyPriceAndDistance false :bDisplayStartedEventUI true :bIsSharedUrgentQuest false :SpecialEventStartedSound "/Game/Sounds/Fort_UI_Sounds/Gameplay/Bounty/Bounty_FailBounty_Cue.Bounty_FailBounty_Cue" :AlertIconBorderColor "(R=1.000000,G=0.456411,B=0.000000,A=1.000000)" :AlertIconBrushSize "(X=0.000000,Y=0.000000)" :bOverrideWidgetPrimaryColor false :OverriddenWidgetPrimaryColor "(R=1.000000,G=0.456411,B=0.000000,A=1.000000)" :ObjectiveCategory "Default"))
  ;[0 0 0] OnUrgentQuestTimeExpired_Event
  (bind _asfort_urgent_quest_item_definition (Utilities|Casting|CastToFortUrgentQuestItemDefinition (Variables|FortTransientQuestGameplayAbility|GetQuestItemDefinition)))
  (if (GameplayTags|Equal(GameplayTag) (Utilities|Struct|BreakUrgentQuestData (Class|FortUrgentQuestItemDefinition|GetUrgentQuestData _asfort_urgent_quest_item_definition)) (Utilities|Struct|BreakUrgentQuestData UrgentQuestData))
    (CallFunction|DisplayMessageFailure _urgentquestdata)
    (CallFunction|SpectatorFailUrgentQuest _urgentquestdata (Variables|Default|GetQuestItemDefEventTag))
    (CallFunction|Cleanup)))

(event Custom|QuestCompleted
  (params (Title :type Text) (Subtitle :type Text) (Special%20Event%20Started%20Sound :type Sound%20Base%20Soft%20Object%20Reference))
  (bind _urgentquestdata (Utilities|Struct|MakeUrgentQuestData :EventTag "(TagName=\"SpecialEvent.HuntingPlayer.IsHunter\")" :VisualStyleTag "(GameplayTags=)" :TotalEventTime 6e+01 :bDisplayTimer true :bQuestCompleteOnTimeExpiration false :bQuickIntro false :bQuickOutro false :bRewardPersist false :bDisplayMultiStepExpanded true :EventTitle Title :EventSubtitle Subtitle :EventDescriptionOverrideTag "(TagName=\"\")" :SortPriority 0 :bHideQuestProgressBar false :bDisplayAllObjectivesIndividually true :QuestProgressTextFormat "LOCGEN_FORMAT_ORDERED(INVTEXT(\"`{{0}} / `{{1}}\"), INVTEXT(\"CurrentProgress\"), INVTEXT(\"MaxProgress\"))" :bReverseProgressBar true :bShowBountyThreatInformation false :bShowBountyPriceAndDistance false :bDisplayStartedEventUI true :bIsSharedUrgentQuest false :SpecialEventStartedSound Special%20Event%20Started%20Sound :AlertIconBorderColor "(R=1.000000,G=0.456411,B=0.000000,A=1.000000)" :AlertIconBrushSize "(X=0.000000,Y=0.000000)" :bOverrideWidgetPrimaryColor false :OverriddenWidgetPrimaryColor "(R=1.000000,G=0.456411,B=0.000000,A=1.000000)" :ObjectiveCategory "Default"))
  ;[0 0 0] Server_QuestCompleted
  (Variables|Default|SetIsQuestCompleted true)
  (CallFunction|DisplayMessageSuccess _urgentquestdata)
  (CallFunction|SpectatorCompleteUrgentQuest _urgentquestdata (Variables|Default|GetQuestItemDefEventTag))
  (CallFunction|Cleanup)
  ;;[0.2135 0.2135 0.2135] This is a time out in case the player does not respond with a ServerEndQuest
  (Utilities|FlowControl|Delay 2.0)
  (CallFunction|ServerEndQuest))

(event Custom|QuestFailed
  (params (Silent :type Boolean) (Title :type Text) (Subtitle :type Text) (Event%20Tag :type Gameplay%20Tag%20Structure) (Special%20Event%20Started%20Sound :type Sound%20Base%20Soft%20Object%20Reference))
  (bind _urgentquestdata (Utilities|Struct|MakeUrgentQuestData Event%20Tag "(GameplayTags=)" true false false false true "" Title Subtitle "" "(TagName=\"\")" true 0 Special%20Event%20Started%20Sound))
  ;[0 0 0] Server_QuestFailed
  (Variables|Default|SetIsQuestCompleted false)
  (if (not Silent)
    (CallFunction|DisplayMessageFailure _urgentquestdata))
  (CallFunction|SpectatorFailUrgentQuest _urgentquestdata (Variables|Default|GetQuestItemDefEventTag))
  (CallFunction|Cleanup)
  ;;[0.2135 0.2135 0.2135] This is a time out in case the player does not respond with a ServerEndQuest
  (Utilities|FlowControl|Delay 2.0)
  (CallFunction|ServerEndQuest))

(event Custom|ServerEndQuest
  ;[0 0 0] Server End Quest with Success/Failure and Clean up
  (Utilities|FlowControl|DoOnce false
    (:Completed
      (if (Variables|Default|GetIsQuestCompleted)
        (CallFunction|CompleteQuest)
        (else
          (CallFunction|FailQuest))))))

(event Custom|EndQueueEvent
  (params (EventTag :type Gameplay%20Tag%20Structure))
  ;[0 0 0] EndQueuing
  (bind _asfort_player_pawn (Utilities|Casting|CastToFortPlayerPawn (Ability|GetAvatarActorfromActorInfo)))
  (Class|FortControllerComponentTransientQuests|EndQueueforPlayerDuelEvent (Class|FortPlayerPawn|GetTransientQuestsComponent _asfort_player_pawn) EventTag))

(event Custom|DisplayMessage_Failure
  (params (UrgentQuestData :type Urgent%20Quest%20Data%20Structure))
  ;[0 0 0] DisplayMessage_Failure
  (Class|AthenaHUDContext|ShowUrgentQuestEndedMessage (BlueprintContext|GetAthenaHUDContext) UrgentQuestData false)
  (CallFunction|ServerEndQuest))

(event Custom|ShowQueuingMessage
  (params (TransientQuestComponent :type Fort%20Controller%20Component%20Transient%20Quests%20Object%20Reference))
  ;[0 0 0] DisplayMessage_Queueing
  (Class|AthenaHUDContext|ShowUrgentQuestMessage (BlueprintContext|GetAthenaHUDContext) (Utilities|Struct|MakeUrgentQuestData "(GameplayTags=)" true false false false true "" "NSLOCTEXT(\"[FBF308A94707D7D5BFB54B980182BE05]\", \"8916BB704DC3177A03A394AA83326F6F\", \"Queued for Battle\")" "NSLOCTEXT(\"[FBF308A94707D7D5BFB54B980182BE05]\", \"4FEC732840BA791C417FAC9616C0046A\", \"Finding Versus Opponent...\")"))
  (Class|FortControllerComponentTransientQuests|StartQueueforPlayerDuelEvent TransientQuestComponent (Utilities|Struct|MakeUrgentQuestData :EventTag "(TagName=\"Athena.Quests.QueuedForDuel\")" :VisualStyleTag "(GameplayTags=)" :TotalEventTime -1.0 :bDisplayTimer true :bQuestCompleteOnTimeExpiration false :bQuickIntro false :bQuickOutro false :bRewardPersist false :bDisplayMultiStepExpanded true :EventTitle "NSLOCTEXT(\"[FBF308A94707D7D5BFB54B980182BE05]\", \"C0CD8D9744C399C09312B0B3E22BFDC6\", \"Waiting for opponent...\")" :EventDescription "NSLOCTEXT(\"[FBF308A94707D7D5BFB54B980182BE05]\", \"C406F75A400FF8D3136165A45FD10F5F\", \"Waiting for opponent...\")" :EventDescriptionOverrideTag "(TagName=\"\")" :SortPriority 0 :bHideQuestProgressBar false :bDisplayAllObjectivesIndividually true :QuestProgressTextFormat "LOCGEN_FORMAT_ORDERED(INVTEXT(\"`{{0}} / `{{1}}\"), INVTEXT(\"CurrentProgress\"), INVTEXT(\"MaxProgress\"))" :bReverseProgressBar true :bShowBountyThreatInformation false :bShowBountyPriceAndDistance false :bDisplayStartedEventUI false :bIsSharedUrgentQuest false :AlertIconBorderColor "(R=1.000000,G=0.456411,B=0.000000,A=1.000000)" :AlertIconBrushSize "(X=0.000000,Y=0.000000)" :bOverrideWidgetPrimaryColor false :OverriddenWidgetPrimaryColor "(R=1.000000,G=0.456411,B=0.000000,A=1.000000)" :ObjectiveCategory "Default")))

(event Custom|DisplayMessage_Success
  (params (UrgentQuestData :type Urgent%20Quest%20Data%20Structure))
  ;[0 0 0] DisplayMessage_Success
  (Class|AthenaHUDContext|ShowUrgentQuestEndedMessage (BlueprintContext|GetAthenaHUDContext) UrgentQuestData true)
  (CallFunction|ServerEndQuest))

(event Custom|OnTargetDied_Bind
  (params (PawnDamageInfo :type Pawn%20Damage%20Data%20Structure%20%28by%20ref%29))
  (bind (_damagedactor _damage _instigatedby) (Utilities|Struct|BreakPawnDamageData PawnDamageInfo))
  (bind _returnvalue (Ability|GetAvatarActorfromActorInfo))
  (bind (_asfort_player_pawn_athena _bsuccess) (Utilities|Casting|CastToFortPlayerPawnAthena _returnvalue))
  (bind (_asplayer_pawn_athena _bsuccess_2) (Utilities|Casting|CastToPlayerPawn_Athena _damagedactor))
  ;[0 0 0] OnTargetDied_Bind
  (if (and _bsuccess_2 _bsuccess)
    (Variables|Default|SetVictimPawn _asplayer_pawn_athena)
    (bind _output_get (Variables|Default|SetKillerPawn (Utilities|Casting|CastToFortPlayerPawnAthena (Pawn|GetControlledPawn _instigatedby))))
    (bind _output_get_2 (Variables|Default|SetOwningPawn _asfort_player_pawn_athena))
    (if (or (Squad|OnSameSquad _output_get _output_get_2) (and (Team|OnSameTeam _output_get _output_get_2) (GameplayTags|HasMatchingGameplayTag _output_get "(TagName=\"Athena.Gameplay.Status.Converted\")")))
      (CallFunction|QuestCompleted "NSLOCTEXT(\"[FBF308A94707D7D5BFB54B980182BE05]\", \"6B6B0B2D48773A7907A1748238E6441D\", \"Battle Won!\")" "NSLOCTEXT(\"[FBF308A94707D7D5BFB54B980182BE05]\", \"CBFFCD4947919E8652FA1A907098C7A6\", \"Opponent eliminated\")" "/Game/Sounds/Fort_UI_Sounds/Gameplay/Bounty/Bounty_Complete_Cue.Bounty_Complete_Cue")
      (elif (Class|AthenaFunctionLibrary|RowtoBool (Variables|Default|GetAllowPoaching))
        (CallFunction|HandlePoacher _output_get)
        (CallFunction|QuestFailed false "NSLOCTEXT(\"[FBF308A94707D7D5BFB54B980182BE05]\", \"41E40FBC4CC78DDC6E70A28BE1A17FD8\", \"Opponent Disqualified! \")" "NSLOCTEXT(\"[FBF308A94707D7D5BFB54B980182BE05]\", \"8567F6F54C5136697A58BDADCC731C97\", \"Opponent eliminated\")" "(TagName=\"SpecialEvent.HuntingPlayer.IsHunter\")" "/Game/Sounds/Fort_UI_Sounds/Gameplay/Bounty/Bounty_Poached_Cue.Bounty_Poached_Cue")
        (CallFunction|BountyPoached)))))

(event Custom|OneMinuteLeft
  ;[0 0 0] One Minute Left Warning
  (Audio|PlaySound2D "/Game/Sounds/Fort_UI_Sounds/Gameplay/Bounty/Bounty_TimeUp_Cue.Bounty_TimeUp_Cue"))

(event Custom|Cleanup
  ;[0 0 0] Cleanup
  (CallFunction|RemoveTrackingCircle)
  (CallFunction|CleanupBountyHunter)
  (if (CallFunction|GetIsPrimaryHunter)
    (CallFunction|CleanupBountyTarget))
  (Variables|Default|SetbHasDuelStarted false)
  (bind _asfort_player_pawn (Utilities|Casting|CastToFortPlayerPawn (Ability|GetAvatarActorfromActorInfo)))
  (GameplayEffects|ApplyGameplayEffectToSelf :self (Ability|GetFortAbilitySystemComponent _asfort_player_pawn) :GameplayEffectClass (Variables|Default|GetHasDueled) :Level 0.0)
  ;[0 0 0] Wait for last damage batch before removing "InDuel" tag.
  (Utilities|FlowControl|Delay 15.0)
  (Ability|RemoveGameplayEffectFromOwnerWithGrantedTags (Variables|Default|GetRemove%20InDuel%20Tag)))

(event Ability|EventOnEndAbility
  (params (bWasCancelled :type Boolean))
  ;[0 0 0]
  (CallFunction|Cleanup))

(event Custom|OnPrimaryHunter_Died
  (params (PawnDamageInfo :type Pawn%20Damage%20Data%20Structure%20%28by%20ref%29))
  (bind (_damagedactor _damage _instigatedby) (Utilities|Struct|BreakPawnDamageData PawnDamageInfo))
  ;[0 0 0] OnPrimaryHunter_Died
  (if (Class|AthenaFunctionLibrary|RowtoBool (Variables|Default|GetAllowPoaching))
    (CallFunction|QuestFailed false "NSLOCTEXT(\"[FBF308A94707D7D5BFB54B980182BE05]\", \"3CAE0EC04B7FD78C4E5CE98D8E01D925\", \"Versus Battle Failed\")" "NSLOCTEXT(\"[FBF308A94707D7D5BFB54B980182BE05]\", \"DBC456414C22308914B16FBB7684E025\", \"Squadmate Eliminated\")")
    (else
      (bind _array_element (Utilities|Array|ForEachLoopwithBreak (CallFunction|GetTrackedBountyHunters))
        (:LoopBody
          (if (Utilities|Equal(Object) (Fortnite|GetFortPlayerControllerfromActor _array_element) _instigatedby)
            (CallFunction|QuestFailed false "NSLOCTEXT(\"[FBF308A94707D7D5BFB54B980182BE05]\", \"7CE4AD3B443FAE0EFE3CCCB6D4D25590\", \"Versus Battle Failed\")" "NSLOCTEXT(\"[FBF308A94707D7D5BFB54B980182BE05]\", \"5384FF1B48968284197D689111DF3C41\", \"Squadmate Eliminated\")")))))))

(event Custom|StartDuelWithBotCandidate
  ;[0 0 0] StartDuelWithBotCandidate
  (CallFunction|StartPlayerDuel))

(event Ability|EventFailedToActivatePassiveAbility
  (params (ActorInfo :type Gameplay%20Ability%20Actor%20Info%20Structure)))

(fn AddTrackingCircleInfo
  (params (Max%20Center%20Offset :type Float%20%28double-precision%29 :default 3.5e+04) (Minimum%20Radius :type Float%20%28double-precision%29 :default 1e+03) (Maximum%20Radius :type Float%20%28double-precision%29 :default 5e+04) (Time%20Between%20Updates%20in%20Seconds :type Float%20%28double-precision%29 :default 5.0) (Radius%20Amount%20to%20Shrink%20Each%20Update :type Float%20%28double-precision%29) (Circle%20Color :type Linear%20Color%20Structure))
  (bind (_owneractor _avataractor _playercontroller) (Utilities|Struct|BreakGameplayAbilityActorInfo (Ability|GetActorInfo)))
  (bind _asfort_player_controller (Utilities|Casting|CastToFortPlayerController _playercontroller)
    (:then
      (bind _outpawn (Class|AbilityFunctions|GetPawnfromPlayerStateZone (CallFunction|GetTrackedHunterBountyTarget))) ; Get the pawn from the player state
      (Variables|Default|SetTracked%20Hunter%20Bounty%20Target%20Pawn _outpawn) ; Store the pawn to remove later, just in case the player respawns\nbefore the pawn is removed (so we don't get the wrong pawn to remove).
      (bind _returnvalue (RadiusTracker|MakeRadiusTrackingInitializer Max%20Center%20Offset Minimum%20Radius Maximum%20Radius Time%20Between%20Updates%20in%20Seconds Radius%20Amount%20to%20Shrink%20Each%20Update Circle%20Color 0 _outpawn (Utilities|Array|MakeArray _asfort_player_controller)))
      (Fortnite|AddActorTrackingInfo (RadiusTracker|GetGameStateRadiusTrackerComponent) _returnvalue))
    (:CastFailed
      (Development|PrintString "Ability Owner Player Controller is not a valid FortPlayerController.  Cannot add tracking info for radius display." false)))
  (return))

(fn CreateUrgentQuestData
  (bind _asfort_urgent_quest_item_definition (Utilities|Casting|CastToFortUrgentQuestItemDefinition (Variables|FortTransientQuestGameplayAbility|GetQuestItemDefinition)))
  (Variables|Default|SetQuestItemDefEventTag (Utilities|Struct|BreakUrgentQuestData (Class|FortUrgentQuestItemDefinition|GetUrgentQuestData _asfort_urgent_quest_item_definition)))
  (Utilities|Struct|SetmembersinUrgentQuestData (Class|FortUrgentQuestItemDefinition|GetUrgentQuestData _asfort_urgent_quest_item_definition) (Math|Float|SelectFloat (Fortnite|GetNumericCVar "Fort.Bounty.TotalQuestTime") (ScalableFloat|GetValueatLevel (Variables|Default|GetTotalQuestTime)) (and (> (Fortnite|GetNumericCVar "Fort.Bounty.TotalQuestTime") 0.0) (Editor|InEditororPIE))) "INVTEXT(\"Testing Subtitle\")" "INVTEXT(\"Testing Subtitle 2\")" "INVTEXT(\"Testing event description\")" 0 0 ""))

(fn HandlePoacher
  (params (Killer :type Fort%20Player%20Pawn%20Object%20Reference))
  (bind _returnvalue (Fortnite|GetFortPlayerControllerfromActor Killer))
  (bind _returnvalue_2 (Quests|GetQuestManager _returnvalue "Athena"))
  (Utilities|IsValid Killer
    (:Is%20Valid
      (Utilities|IsValid _returnvalue
        (:Is%20Valid
          (Utilities|IsValid _returnvalue_2
            (:Is%20Valid
              (Quest|SendComplexCustomStatEvent _returnvalue_2 0 "(GameplayTags=)" "(GameplayTags=((TagName=\"Athena.Quests.Bounty.Poacher\")))" 1))))))))

(fn OnRep_bQueuedForDuel)

(fn RemoveTrackingCircle
  (bind (_owneractor _avataractor _playercontroller) (Utilities|Struct|BreakGameplayAbilityActorInfo (Ability|GetActorInfo)))
  (bind _asfort_player_controller (Utilities|Casting|CastToFortPlayerController _playercontroller))
  (RadiusTracker|RemoveTrackingforActorfromPlayer (RadiusTracker|GetGameStateRadiusTrackerComponent) (Variables|Default|GetTracked%20Hunter%20Bounty%20Target%20Pawn) _asfort_player_controller "Team"))
